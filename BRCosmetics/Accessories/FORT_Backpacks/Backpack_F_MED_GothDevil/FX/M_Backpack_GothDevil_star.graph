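# Translucent, two-sided unlit Niagara material for the star effect on the GothDevil backpack.
# Vertices displace along normals by panning noise plus a tiled mask, scaled by dynamic parameters; SplineThicken adds only as Use_SplineThicken drops below its default 1.
# Particle color drives emissive.
# Opacity multiplies particle alpha, a pannable line texture, V-edge falloff and view-angle fade, with depth fade skipped on forward and mobile paths.
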Material /BRCosmetics/Accessories/FORT_Backpacks/Backpack_F_MED_GothDevil/FX/M_Backpack_GothDevil_star
graph {
 properties {
  bAutomaticallySetUsageInEditor=false
  bUsedWithNiagaraMeshParticles=true
  bUsedWithNiagaraRibbons=true
  bUsedWithNiagaraSprites=true
  blendMode=BLEND_Translucent
  shadingModel=MSM_Unlit
  twoSided=true
  usedShadingModels=Unlit
 }
 nodes [
  n1 = TextureCoordinate()
  n2 = AppendVector(A=Constant() {r=1}, B=Constant())
  n3 = DynamicParameter() {defaultValue=<rgba 10 -0.5 1 1>, paramNames [StarWPO_Intensity, NoiseTexture_SpeedX, Param3, Param4]}
  n4 = Constant()
  n5 = TextureSampleParameter2D(UVs=Panner(Coordinate=Multiply(A=@n1, B=@n2), Speed=AppendVector(A=@n3.NoiseTexture_SpeedX, B=@n4)) {speedX=-0.5}) {group=Noise, parameterName=NoiseTexture, texture=/BRCosmetics/Weapons/FORT_Melee/Pickaxe_CoatCheck/FX/T_Pickaxe_CoatCheck_Noise}
  n6 = LinearInterpolate(B=ScalarParameter() {defaultValue=0.4, parameterName=NoiseTexture_Lerp}, Alpha=@n5.R) {constB=0.4}
  n7 = TextureSample(UVs=TextureCoordinate() {uTiling=5}) {texture=&/T_Backpack_GothDevil_Mask}
  n8 = Multiply(
   A=Multiply(A=VertexNormalWS(), B=Add(A=@n6, B=@n7.R))
   B=DynamicParameter() {defaultValue=<rgba 10 -0.5 1 1>, paramNames [StarWPO_Intensity, NoiseTexture_SpeedX, Param3, Param4]}.StarWPO_Intensity
  )
  n9 = ScalarParameter() {defaultValue=20, parameterName=Spline_Base}
  n10 = ScalarParameter() {defaultValue=20, parameterName=Spline_Tip}
  n11 = StaticBool()
  LinearInterpolate(
   A=Add(
    A=@n8
    B=MaterialFunctionCall(WidthBase=@n9, WidthTip=@n10, Expand U or V UV Channel=@n11) {materialFunction=&SplineThicken}.WorldPositionOffset
   )
   B=@n8
   Alpha=ScalarParameter() {defaultValue=1, parameterName=Use_SplineThicken}
  )
  n12 = ParticleColor()
  MaterialFunctionCall(Emissive=Multiply(A=@n12.RGB, B=ScalarParameter() {defaultValue=1, parameterName=All_Light})) {materialFunction=&MF_FX_EmissiveAdjustments}
  ScalarParameter() {defaultValue=10, parameterName=WPO_Intensity}
  n13 = Multiply(
   A=TextureCoordinate()
   B=AppendVector(A=ScalarParameter() {defaultValue=1, parameterName=BaseTextureU}, B=ScalarParameter() {defaultValue=1, parameterName=BaseTextureV})
  )
  n14 = AppendVector(A=ScalarParameter() {parameterName=BaseTexture_SpeedX}, B=ScalarParameter() {parameterName=BaseTexture_SpeedY})
  n15 = Power(
   Base=TextureSampleParameter2D(UVs=Panner(Coordinate=@n13, Speed=@n14)) {parameterName=BaseTexture, texture=/Game/Weapons/FORT_Rifles/Effects/Textures/T_FX_ThinLine_BoomB_01}.R
   Exp=ScalarParameter() {defaultValue=1, parameterName=BaseTexturePower}
  )
  n16 = ComponentMask(None=TextureCoordinate()) {g=true}
  n17 = OneMinus(None=@n16)
  n18 = Multiply(A=Power(Base=Multiply(A=@n16) {constB=2}), B=Power(Base=Multiply(A=@n17) {constB=2}))
  n19 = Multiply(
   A=Multiply(A=Multiply(A=@n12.A, B=@n15), B=Saturate(None=@n18))
   B=MaterialFunctionCall("FadeSensitivity (optional)"=ScalarParameter() {parameterName=MeshFade}) {materialFunction=&FadeBasedOnViewAngle}.Result
  )
  n20 = DepthFade(Opacity=@n19, FadeDistance=ScalarParameter() {defaultValue=5, parameterName=DepthFade})
  ShadingPathSwitch(Default=@n20, Deferred=@n20, Forward=@n19, Mobile=@n19)
 ]
}
